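# Physics asset for the RetroPhotographer backpack cosmetic, simulating small dangling paper pieces attached to the pack.
# A large capsule body covers the pack, and small capsules cover each dyn_paper bone.
# Each paper hangs from the pack by its own angle-limited constraint, except paper 01, which is a two-link root-to-end chain.
# Tight limits let most pieces swing mainly on one axis, while paper 04 mostly twists.
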PhysicsAsset /BRCosmetics/Accessories/FORT_Backpacks/Backpack_F_MED_RetroPhotographer/Meshes/F_MED_RetroPhotographer_Pack_Physics
bodies {
 dyn_paper_01_end [{shapeType=Capsule, length=0.505, radius=0.505, shapeName=dyn_paper_01_end_capsule, center=<-0.0187 0 0>}]
 dyn_paper_01_root [{shapeType=Capsule, length=0.505, radius=0.505, shapeName=dyn_paper_01_root_capsule}]
 dyn_paper_02 [{
   shapeType=Capsule
   length=1.6931
   radius=0.5
   shapeName=dyn_paper_02_capsule
   center=<0.0139 -0.0004 0.0034>
   rotation=<rot -8.7912 -89.2869 -103.8459>
  }]
 dyn_paper_03 [{shapeType=Capsule, length=3.5516, radius=0.505, shapeName=dyn_paper_03_capsule, center=<-0.1 0 0>, rotation=<rot 0 0 90>}]
 dyn_paper_04 [{shapeType=Capsule, length=2.3452, radius=0.505, shapeName=dyn_paper_04_capsule, center=<0 0.0179 0>, rotation=<rot 90 -90 -90.0005>}]
 dyn_paper_05 [{shapeType=Capsule, length=4.9162, radius=0.8722, shapeName=dyn_paper_05_capsule, center=<0.2899 0 0>, rotation=<rot 0 0 -90.0001>}]
 pack [{shapeType=Capsule, length=16.6102, radius=10.7799, shapeName=pack_capsule, rotation=<rot 90 -90 -90.0001>}]
}
constraints [
 {bone1Name=dyn_paper_02, bone2Name=pack, swing1LimitDegrees=1.4, swing2LimitDegrees=11.32, twistLimitDegrees=2}
 {bone1Name=dyn_paper_03, bone2Name=pack, swing1LimitDegrees=1.3, swing2LimitDegrees=45, twistLimitDegrees=2}
 {bone1Name=dyn_paper_04, bone2Name=pack, swing1LimitDegrees=1, swing2LimitDegrees=1, twistLimitDegrees=25}
 {bone1Name=dyn_paper_05, bone2Name=pack, swing1LimitDegrees=1, swing2LimitDegrees=40, twistLimitDegrees=1}
 {bone1Name=dyn_paper_01_root, bone2Name=pack, swing1LimitDegrees=2, swing2LimitDegrees=26, twistLimitDegrees=2}
 {bone1Name=dyn_paper_01_end, bone2Name=dyn_paper_01_root, swing1LimitDegrees=2, swing2LimitDegrees=28, twistLimitDegrees=1}
]
